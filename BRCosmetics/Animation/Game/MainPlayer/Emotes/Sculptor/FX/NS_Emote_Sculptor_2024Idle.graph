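# Niagara idle effect for the Sculptor 2024 emote: pale cyan sparkles around a translucent blue glow.
# One emitter continuously spawns small, short-lived sprites in a box surface band, growing then shrinking them by curve.
# The other bursts a persistent large sprite with two renderers, using separate perpendicular facing vectors to suggest a vertical cross.
# Its alpha fades in over particle age and pulses on a curve indexed by normalized system loop age.
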NiagaraSystem /BRCosmetics/Animation/Game/MainPlayer/Emotes/Sculptor/FX/NS_Emote_Sculptor_2024Idle
dependencies {usedDataInterfaces [&NiagaraDataInterfaceCurve], usedDynamicInputs [], usedModules [&SystemState]}
emitters {
 Empty {
  emitterUpdateScript [
   &EmitterState()
   &SpawnRate(SpawnRate=10, Use Spawn Probability=false)
  ]
  particleSpawnScript [
   &InitializeParticle(
    Color Maximum=<rgba 0.7584 1 0.9701 1>
    Color Minimum=<rgba 0.6008 0.9501 1 1>
    Color Mode=::NewEnumerator2
    Lifetime Max=0.7
    Lifetime Min=0.4
    Lifetime Mode=::NewEnumerator1
    Sprite Size Mode=::NewEnumerator1
    Uniform Sprite Size Max=14
    Uniform Sprite Size Min=6
    UsePositionOffset=false
    Write Material Random=true
   )
   &ShapeLocation(
    Box Size=<60 15 80>
    MaskSpawnGroup=false
    Override Seed=false
    Shape Primitive=::NewEnumerator2
    Surface Only=true
    Surface Only Band Thickness=20
    UseOwnerScale=true
   )
  ]
  particleUpdateScript [
   &ParticleState()
   &ScaleSpriteSize(Uniform Curve Sprite Scale=DataInterface() {propertyValues {Curve {keys [
        {tangentMode=RCTM_User, time=0, value=0}
        {interpMode=RCIM_Cubic, leaveTangent=1.468, tangentMode=RCTM_Break, time=0.2, value=0.65}
        {interpMode=RCIM_Cubic, tangentMode=RCTM_Break, time=0.6, value=0.95}
        {arriveTangent=-3.5336, tangentMode=RCTM_User, time=1, value=0.1}
       ]}}})
  ]
  rendererClasses [&NiagaraSpriteRendererProperties]
  renderers [{rendererClass=&NiagaraSpriteRendererProperties}]
 }
 Empty001 {
  emitterUpdateScript [
   &EmitterState(Life Cycle Mode=::NewEnumerator1, Loop Behavior=::NewEnumerator1, Loop Duration=1, UseLoopDelay=false)
   &SpawnBurst_Instantaneous(Use Spawn Probability=false, UseLoopCountLimit=false)
  ]
  particleSpawnScript [
   &InitializeParticle(
    Color=<rgba 0.1633 0.7366 1 0.6>
    Lifetime=5
    Sprite Size Mode=::NewEnumerator3
    Uniform Sprite Size=120
    UsePositionOffset=false
    Write Material Random=true
   )
   /SetVariables(Particles.SpriteAlignment=<0 0 1>, Particles.SpriteFacing=<0 1 0>) {isSetParameters=true}
   /SetVariables(Particles.Render2SpriteAlignment:Vector3f=<0 0 1>, Particles.Render2SpriteFacing:Vector3f=<1 0 0>) {isSetParameters=true}
  ]
  particleUpdateScript [
   &ParticleState(Kill Particles When Lifetime Has Elapsed=false)
   &ScaleColor(
    Scale Alpha=&FloatFromCurve(FloatCurve=DataInterface() {propertyValues {Curve {keys [<at 0 0>,<at 0.3 1>]}}})
    ScaleA=true
    ScaleRGB=true
   ) {module=ScaleColor001}
   &ScaleColor(
    Scale Alpha=&FloatFromCurve(
     FloatCurve=DataInterface() {propertyValues {Curve {keys [<at 0 0.6>,<at 0.1714 1>,<at 0.2714 0.6029>,<at 0.4616 1.0265>,<at 0.6176 0.6529>,<at 0.8033 0.9647>,<at 0.9965 0.6>]}}}
     CurveIndex=$System.NormalizedLoopAge
    )
    ScaleA=true
    ScaleRGB=true
   )
  ]
  rendererClasses [&NiagaraSpriteRendererProperties]
  renderers [
   {rendererClass=&NiagaraSpriteRendererProperties}
   {rendererClass=&NiagaraSpriteRendererProperties, rendererIndex=1}
  ]
 }
}
summary {
 systemName=NS_Emote_Sculptor_2024Idle
 emitters [
  {emitterName=Empty, rendererClasses [&NiagaraSpriteRendererProperties]}
  {emitterName=Empty001, rendererClasses [&NiagaraSpriteRendererProperties]}
 ]
 userVariables []
}
